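# Physics asset for the tail of the large female BunnyBR player body mesh.
# It has capsule bodies on two bones: a tiny one on base_tail and a longer, offset one along dyn_tail.
# A single constraint joins dyn_tail to base_tail with narrow, equal swing limits and a twist limit set to zero degrees.
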PhysicsAsset /Game/Characters/Player/Female/Large/Bodies/F_LRG_BunnyBR/Meshes/F_LRG_BunnyBR_tail_Physics
bodies {
 base_tail [{shapeType=Capsule, length=0.505, radius=0.505, shapeName=base_tail_capsule, rotation=<rot 0 -90 0>}]
 dyn_tail [{shapeType=Capsule, length=10.5444, radius=1.6107, shapeName=dyn_tail_capsule, center=<5.9454 0 0>, rotation=<rot 90 0 0>}]
}
constraints [{bone1Name=dyn_tail, bone2Name=base_tail, swing1LimitDegrees=15, swing2LimitDegrees=15, twistLimitDegrees=0}]
