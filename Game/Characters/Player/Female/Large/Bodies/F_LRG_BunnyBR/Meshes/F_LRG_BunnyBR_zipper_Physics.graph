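# Physics asset for the zipper mesh of the large female BunnyBR player body, apparently for dangling zipper motion.
# It has two small capsule bodies, one on base_zipper and a longer, offset one on dyn_zipper.
# A single constraint joins dyn_zipper to base_zipper, with a narrow limit on swing 2 and zero limits on swing 1 and twist, suggesting hinge-like movement.
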PhysicsAsset /Game/Characters/Player/Female/Large/Bodies/F_LRG_BunnyBR/Meshes/F_LRG_BunnyBR_zipper_Physics
bodies {
 base_zipper [{
   shapeType=Capsule
   length=1.0497
   radius=0.7428
   shapeName=base_zipper_capsule
   center=<0.2515 -0.0002 -0.0878>
   rotation=<rot 0 91.3491 -87.4223>
  }]
 dyn_zipper [{shapeType=Capsule, length=1.8959, radius=1.0293, shapeName=dyn_zipper_capsule, center=<2.2067 0 0>, rotation=<rot 90 0 0>}]
}
constraints [{bone1Name=dyn_zipper, bone2Name=base_zipper, swing1LimitDegrees=0, swing2LimitDegrees=20, twistLimitDegrees=0}]
