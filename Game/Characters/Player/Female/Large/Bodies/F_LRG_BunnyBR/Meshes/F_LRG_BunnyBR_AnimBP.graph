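# Animation Blueprint for the F_LRG_BunnyBR player body, adding procedural and physics secondary motion for its tail and zipper.
# The pose copied from the attached parent mesh passes through a Control Rig, a weighted base_zipper constraint, then tail and zipper RigidBody simulations in pelvis and spine_05 space.
# The event graph updates GravityOverride and backbling state, and on customization load toggles the "Hide Element 01" material parameter for long backblings (capes).
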AnimBlueprint /Game/Characters/Player/Female/Large/Bodies/F_LRG_BunnyBR/Meshes/F_LRG_BunnyBR_AnimBP
target_skeleton=&/F_LRG_BunnyBR_Skeleton
graphs [
 AnimGraph {nodes [
   n1 = Control Rig(Source=Copy Pose From Mesh() {bUseAttachedParent=true}) {
    alphaBoolBlend {blendOption=Linear}
    controlRigAssetReference {blueprintRigClass=&/F_LRG_BunnyBR_CtrlRig.F_LRG_BunnyBR_CtrlRig_C}
   }
   n2 = "Constraint - base_zipper (PPPPPP)"!(
    ComponentPose=Local To Component(LocalPose=@n1)
    Alpha=1
    ConstraintWeights_0=0.056
    ConstraintWeights_1:float=0.056
    ConstraintWeights_2:float=0.079
    ConstraintWeights_3:float=0.331
    ConstraintWeights_4:float=0.462
    ConstraintWeights_5:float=0.017
   ) -> (Pose) {
    alphaBoolBlend {blendOption=Linear}
    boneToModify {boneName=base_zipper}
    constraintSetup [
     {targetBone {boneName=clavicle_l}}
     {targetBone {boneName=clavicle_r}}
     {targetBone {boneName=spine_03}}
     {targetBone {boneName=spine_04}}
     {targetBone {boneName=spine_05}}
     {targetBone {boneName=neck_01}}
    ]
    constraintWeights [1,1,1,1,1,1]
   }
   n3 = RigidBody(ComponentPose=@n2) {
    bClampLinearTranslationLimitToRefPose=true
    bOverrideWorldGravity=true
    overridePhysicsAsset=&/F_LRG_BunnyBR_tail_Physics
    simulationSpace=BaseBoneSpace
    alphaBoolBlend {blendOption=Linear}
    baseBoneRef {boneName=pelvis}
    componentAppliedLinearAccClamp=<6000 6000 6000>
    componentLinearAccScale=<0.6 0.6 0.6>
    componentLinearVelScale=<0.6 0.6 0.6>
    overrideWorldGravity {z=-980}
   }
   n4 = RigidBody(ComponentPose=@n3) {
    bClampLinearTranslationLimitToRefPose=true
    bOverrideWorldGravity=true
    overridePhysicsAsset=&/F_LRG_BunnyBR_zipper_Physics
    simulationSpace=BaseBoneSpace
    alphaBoolBlend {blendOption=Linear}
    baseBoneRef {boneName=spine_05}
    componentAppliedLinearAccClamp=<6000 6000 6000>
    componentLinearAccScale=<0.6 0.6 0.6>
    componentLinearVelScale=<0.6 0.6 0.6>
    overrideWorldGravity {z=-980}
   }
   Output Pose(Result=Component To Local(ComponentPose=@n4)) {layerGroup=None}
  ]}
 EventGraph {nodes [
   Event BlueprintUpdateAnimation()
   > n1 = Sequence!(execute) -> (then_0, then_1, then_2:Exec)
   > n2 = Gravity Override State Function DEPRECATED(CharacterPartAnimInstance=Self-Reference(), GameTimeInSeconds=GetGameTimeInSeconds(), GravityOverrideParams=Get Gravity Override Params())
   > Set GravityOverride(GravityOverride=@n2)
   Set isWearingBackbling(execute=@n1.then_1, isWearingBackbling=CheckForBackbling())
   Event OnCharacterCustomizationLoaded()
   > n3 = Sequence!(execute) -> (then_0, then_1)
   n4 = CheckForBackbling()
   > Set isWearingBackbling(isWearingBackbling=@n4)
   > Set Is Wearing Cape(Is Wearing Cape=@n4.bIsBackblingLong)
   n5 = Get Dynamic Material Instance(execute=@n3.then_1, SkeletalMeshComponent=GetPartSkeletalMeshComponent())
   > n6 = Branch(Condition=Get Is Wearing Cape() -> (Is Wearing Cape:bool))
   > SetScalarParameterValue(self=@n5, ParameterName=Hide Element 01, Value=1)
   SetScalarParameterValue(execute=@n6.else, self=@n5, ParameterName=Hide Element 01, Value=0)
  ]}
]
variables [
 BodyMesh {type=SkeletalMeshComponent}
 cape and tail combo {type=boolean}
 characterHasTail {type=boolean}
 Gravity Override Params {type=GravityOverrideParamsStruct}
 GravityOverride {type=Vector}
 Is Pickaxing Alpha {type=number}
 Is Wearing Cape {type=boolean}
 isWearingBackbling {type=boolean}
 PawnOwner {type=PlayerPawn_Athena_C}
 TailMesh {type=SkeletalMeshComponent}
]
